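# Sound cue for the fifth LifeTape noise sound, filed under the NightNight quest sounds in Figment Season 10.
# A single wave player feeds the root directly, with no randomization, mixing or modulation.
# It uses the shared quest-interactable hint attenuation with a limited audible range, and is assigned to the UI sound class.
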SoundCue /Figment_S10/Athena/Sounds/Quests/NightNight/LifeTape_noise_05_Cue
root=WavePlayer_0
nodes [{id=WavePlayer_0, properties {soundWaveAssetPtr=&/LifeTape_noise_05}}]
properties {
 attenuationSettings=/Game/Sounds/Attenuation/Fort_QuestInteractables_Hint
 maxDistance=1300
 soundClassObject=&Fort_UI
 volumeMultiplier=1
}
